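# Sound cue for the soft cloth flap foley in the Temple emote.
# A Random node picks one of three equally weighted light cotton flap waves, and a Modulator varies its playback within a lowered pitch range.
# It uses emote attenuation and foley concurrency, but sits in the emote music sound class and lists a send to the music audio-analysis submix.
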SoundCue /BRCosmetics/Sounds/Emotes/Temple/Emote_Temple_ClothFlap_Soft_Cue
root=Modulator_0
nodes [
 {id=Modulator_0, children [Random_1], properties {pitchMax=0.8, pitchMin=0.7}}
 {id=Random_1, children [WavePlayer_5,WavePlayer_6,WavePlayer_7], properties {bShouldExcludeFromBranchCulling=true, weights [1,1,1]}}
 {id=WavePlayer_5, properties {soundWaveAssetPtr=&/cloth_cotton_medium_flap_light_06}}
 {id=WavePlayer_6, properties {soundWaveAssetPtr=&/cloth_cotton_medium_flap_light_07}}
 {id=WavePlayer_7, properties {soundWaveAssetPtr=&/cloth_cotton_medium_flap_light_08}}
]
properties {
 attenuationSettings=&Fort_Emotes_Athena_Att
 maxDistance=1600
 soundClassObject=&Fort_Music_Emote
 volumeMultiplier=0.8
 concurrencySet [/Game/Sounds/Concurrency_Rules/Athena_EmoteFoley_Conc]
 soundSubmixSends [{minSendLevel=0, soundSubmix=&Fortnite_AudioAnalysis_Music}]
}
