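# Sound cue that plays foot scuff foley for the Temple emote, using gritty boot-on-stone samples.
# A Random node picks one of four equally weighted waves and feeds a Modulator at default settings.
# It uses emote attenuation and the emote music sound class, with a send to the music audio analysis submix.
# An emote foley concurrency set is assigned, alongside override settings that stop the oldest instance.
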SoundCue /BRCosmetics/Sounds/Emotes/Temple/Emote_Temple_FootScuffs_Cue
root=Modulator_0
nodes [
 {id=Modulator_0, children [Random_1]}
 {id=Random_1, children [WavePlayer_0, WavePlayer_1, WavePlayer_10, WavePlayer_11], properties {weights [1,1,1,1]}}
 {id=WavePlayer_0, properties {soundWaveAssetPtr=&/footstep_boot_stone_gritty_scuff_10}}
 {id=WavePlayer_1, properties {soundWaveAssetPtr=&/footstep_boot_stone_gritty_scuff_03}}
 {id=WavePlayer_10, properties {soundWaveAssetPtr=&/footstep_boot_stone_gritty_scuff_04}}
 {id=WavePlayer_11, properties {soundWaveAssetPtr=&/footstep_boot_stone_gritty_scuff_06}}
]
properties {
 attenuationSettings=&Fort_Emotes_Athena_Att
 maxDistance=1600
 soundClassObject=&Fort_Music_Emote
 volumeMultiplier=0.65
 concurrencyOverrides {maxCount=3, resolutionRule=StopOldest}
 concurrencySet [/Game/Sounds/Concurrency_Rules/Athena_EmoteFoley_Conc]
 soundSubmixSends [{minSendLevel=0, soundSubmix=&Fortnite_AudioAnalysis_Music}]
}
