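# Footstep SoundCue for the Temple emote, playing gritty boot-on-stone steps with emote attenuation, the emote music sound class and a music analysis submix send.
# Each step mixes a random heel sample with a quieter random toe sample after a short random delay, under a Modulator.
# A quality-level switch gives the first two slots larger sample pools and the last two a two-sample-per-part version with a louder toe.
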SoundCue /BRCosmetics/Sounds/Emotes/Temple/Emote_Temple_FootSteps_Cue
root=Modulator_0
nodes [
 {id=Delay_0, children [Random_2], properties {delayMax=0.05, delayMin=0.025}}
 {id=Delay_1, children [Random_4], properties {delayMax=0.05, delayMin=0.025}}
 {id=Mixer_0, children [Random_0,Delay_0], properties {inputVolume [1,0.35]}}
 {id=Mixer_1, children [Random_3,Delay_1], properties {inputVolume [1,0.5]}}
 {id=Modulator_0, children [QualityLevel_0]}
 {id=QualityLevel_0, children [Mixer_0,Mixer_0,Mixer_1,Mixer_1]}
 {id=Random_0, children [WavePlayer_2,WavePlayer_3,WavePlayer_5,WavePlayer_6,WavePlayer_7], properties {weights [1,1,1,1,1]}}
 {id=Random_2, children [WavePlayer_8, WavePlayer_9, WavePlayer_16, WavePlayer_17], properties {weights [1,1,1,1]}}
 {id=Random_3, children [WavePlayer_19, WavePlayer_22], properties {bShouldExcludeFromBranchCulling=true, weights [1,1]}}
 {id=Random_4, children [WavePlayer_24, WavePlayer_27], properties {bShouldExcludeFromBranchCulling=true, weights [1,1]}}
 {id=WavePlayer_16, properties {soundWaveAssetPtr=&/footstep_boot_stone_gritty_run_toe_03}}
 {id=WavePlayer_17, properties {soundWaveAssetPtr=&/footstep_boot_stone_gritty_run_toe_06}}
 {id=WavePlayer_19, properties {soundWaveAssetPtr=&/footstep_boot_stone_gritty_run_heel_10}}
 {id=WavePlayer_2, properties {soundWaveAssetPtr=&footstep_boot_stone_gritty_run_heel_10}}
 {id=WavePlayer_22, properties {soundWaveAssetPtr=&/footstep_boot_stone_gritty_run_heel_08}}
 {id=WavePlayer_24, properties {soundWaveAssetPtr=&/footstep_boot_stone_gritty_run_toe_01}}
 {id=WavePlayer_27, properties {soundWaveAssetPtr=&footstep_boot_stone_gritty_run_toe_06}}
 {id=WavePlayer_3, properties {soundWaveAssetPtr=&/footstep_boot_stone_gritty_run_heel_02}}
 {id=WavePlayer_5, properties {soundWaveAssetPtr=&/footstep_boot_stone_gritty_run_heel_05}}
 {id=WavePlayer_6, properties {soundWaveAssetPtr=&footstep_boot_stone_gritty_run_heel_08}}
 {id=WavePlayer_7, properties {soundWaveAssetPtr=&/footstep_boot_stone_gritty_run_heel_09}}
 {id=WavePlayer_8, properties {soundWaveAssetPtr=&footstep_boot_stone_gritty_run_toe_01}}
 {id=WavePlayer_9, properties {soundWaveAssetPtr=&/footstep_boot_stone_gritty_run_toe_02}}
]
properties {
 attenuationSettings=&Fort_Emotes_Athena_Att
 maxDistance=1600
 soundClassObject=&Fort_Music_Emote
 volumeMultiplier=0.8
 concurrencyOverrides {maxCount=3, resolutionRule=StopOldest}
 concurrencySet [/Game/Sounds/Concurrency_Rules/Athena_EmoteFoley_Conc]
 soundSubmixSends [{minSendLevel=0, soundSubmix=&Fortnite_AudioAnalysis_Music}]
}
