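# Sound cue for the hand-plant foley of the Temple emote, a stone hand impact with a trailing scuff.
# Each branch mixes a random plant with a quieter, slightly delayed, higher-pitched random scuff, under an overall randomized upward pitch shift.
# A quality-level node gives two slots the full wave pools and the other two a reduced pool of two plants and two scuffs.
# Emote attenuation and stop-oldest concurrency govern playback.
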SoundCue /BRCosmetics/Sounds/Emotes/Temple/Emote_Temple_HandPlant_Cue
root=Modulator_0
nodes [
 {id=Delay_1, children [Modulator_1], properties {delayMax=0.04, delayMin=0.02}}
 {id=Delay_2, children [Modulator_2], properties {delayMax=0.04, delayMin=0.02}}
 {id=Mixer_1, children [Random_1,Delay_1], properties {inputVolume [1,0.4]}}
 {id=Mixer_2, children [Random_4,Delay_2], properties {inputVolume [1,0.4]}}
 {id=Modulator_0, children [QualityLevel_0], properties {pitchMax=1.25, pitchMin=1.15}}
 {id=Modulator_1, children [Random_3], properties {pitchMax=1.35, pitchMin=1.25}}
 {id=Modulator_2, children [Random_6], properties {pitchMax=1.35, pitchMin=1.25}}
 {id=QualityLevel_0, children [Mixer_1,Mixer_1,Mixer_2,Mixer_2]}
 {id=Random_1, children [WavePlayer_0, WavePlayer_1, WavePlayer_10, WavePlayer_11, WavePlayer_12], properties {weights [1,1,1,1,1]}}
 {
  id=Random_3
  children [WavePlayer_13, WavePlayer_14, WavePlayer_15, WavePlayer_19, WavePlayer_20, WavePlayer_21, WavePlayer_22]
  properties {weights [1,1,1,1,1,1,1]}
 }
 {id=Random_4, children [WavePlayer_23, WavePlayer_25], properties {bShouldExcludeFromBranchCulling=true, weights [1,1]}}
 {id=Random_6, children [WavePlayer_28, WavePlayer_29], properties {bShouldExcludeFromBranchCulling=true, weights [1,1]}}
 {id=WavePlayer_0, properties {soundWaveAssetPtr=&/hand_stone_plant_01}}
 {id=WavePlayer_1, properties {soundWaveAssetPtr=&/hand_stone_plant_02}}
 {id=WavePlayer_10, properties {soundWaveAssetPtr=&/hand_stone_plant_03}}
 {id=WavePlayer_11, properties {soundWaveAssetPtr=&/hand_stone_plant_04}}
 {id=WavePlayer_12, properties {soundWaveAssetPtr=&/hand_stone_plant_05}}
 {id=WavePlayer_13, properties {soundWaveAssetPtr=&/hand_stone_scuff_01}}
 {id=WavePlayer_14, properties {soundWaveAssetPtr=&/hand_stone_scuff_02}}
 {id=WavePlayer_15, properties {soundWaveAssetPtr=&/hand_stone_scuff_03}}
 {id=WavePlayer_19, properties {soundWaveAssetPtr=&/hand_stone_scuff_04}}
 {id=WavePlayer_20, properties {soundWaveAssetPtr=&/hand_stone_scuff_05}}
 {id=WavePlayer_21, properties {soundWaveAssetPtr=&/hand_stone_scuff_06}}
 {id=WavePlayer_22, properties {soundWaveAssetPtr=&/hand_stone_scuff_07}}
 {id=WavePlayer_23, properties {soundWaveAssetPtr=&hand_stone_plant_01}}
 {id=WavePlayer_25, properties {soundWaveAssetPtr=&hand_stone_plant_03}}
 {id=WavePlayer_28, properties {soundWaveAssetPtr=&hand_stone_scuff_01}}
 {id=WavePlayer_29, properties {soundWaveAssetPtr=&hand_stone_scuff_02}}
]
properties {
 attenuationSettings=&Fort_Emotes_Athena_Att
 maxDistance=1600
 soundClassObject=&Fort_Music_Emote
 volumeMultiplier=0.85
 concurrencyOverrides {maxCount=3, resolutionRule=StopOldest}
 concurrencySet [/Game/Sounds/Concurrency_Rules/Athena_EmoteFoley_Conc]
 soundSubmixSends [{minSendLevel=0, soundSubmix=&Fortnite_AudioAnalysis_Music}]
}
