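# Sound cue for the roll impact in the Temple emote, layering a body-fall thud with a roll sound.
# A mixer blends one of two random thuds, at reduced volume, with a briefly delayed, pitch-varied roll at full volume.
# An outer modulator randomly raises the pitch of the whole mix.
# It uses emote attenuation, the emote music sound class, stop-oldest concurrency and a music-analysis submix send.
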SoundCue /BRCosmetics/Sounds/Emotes/Temple/Emote_Temple_RollImpact_Cue
root=Modulator_0
nodes [
 {id=Delay_0, children [WavePlayer_3], properties {delayMax=0.05, delayMin=0.05}}
 {id=Mixer_1, children [Random_0,Modulator_2], properties {inputVolume [0.35,1]}}
 {id=Modulator_0, children [Mixer_1], properties {pitchMax=1.25, pitchMin=1.15}}
 {id=Modulator_2, children [Delay_0], properties {pitchMax=1.1, pitchMin=0.9}}
 {id=Random_0, children [WavePlayer_4,WavePlayer_2], properties {weights [1,1]}}
 {id=WavePlayer_2, properties {soundWaveAssetPtr=/Game/Athena/Sounds/Emotes/Temple/bodyfall_generic_thud_06}}
 {id=WavePlayer_3, properties {soundWaveAssetPtr=&/bodyfall_roll_01}}
 {id=WavePlayer_4, properties {soundWaveAssetPtr=/Game/Athena/Sounds/Emotes/Temple/bodyfall_generic_thud_04}}
]
properties {
 attenuationSettings=&Fort_Emotes_Athena_Att
 maxDistance=1600
 soundClassObject=&Fort_Music_Emote
 volumeMultiplier=0.7
 concurrencyOverrides {maxCount=3, resolutionRule=StopOldest}
 concurrencySet [/Game/Sounds/Concurrency_Rules/Athena_EmoteFoley_Conc]
 soundSubmixSends [{minSendLevel=0, soundSubmix=&Fortnite_AudioAnalysis_Music}]
}
