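# Sound cue for a high swoosh effect in the Temple emote.
# A Random node picks one of three equally weighted fast swoosh waves, and a Modulator pitches it well down with slight pitch and volume variation.
# It uses emote attenuation, the emote music sound class and an emote foley concurrency set.
# A submix send targets the music audio analysis submix.
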SoundCue /BRCosmetics/Sounds/Emotes/Temple/Emote_Temple_SwhooshHigh_Cue
root=Modulator_0
nodes [
 {id=Modulator_0, children [Random_1], properties {pitchMax=0.7, pitchMin=0.55, volumeMin=0.65}}
 {id=Random_1, children [WavePlayer_1, WavePlayer_10, WavePlayer_12], properties {bShouldExcludeFromBranchCulling=true, weights [1,1,1]}}
 {id=WavePlayer_1, properties {soundWaveAssetPtr=&/swoosh_high_fast_d_02}}
 {id=WavePlayer_10, properties {soundWaveAssetPtr=&/swoosh_high_fast_d_03}}
 {id=WavePlayer_12, properties {soundWaveAssetPtr=/Game/Athena/Sounds/Emotes/Temple/swoosh_high_fast_d_01}}
]
properties {
 attenuationSettings=&Fort_Emotes_Athena_Att
 maxDistance=1600
 soundClassObject=&Fort_Music_Emote
 volumeMultiplier=0.35
 concurrencyOverrides {maxCount=3, resolutionRule=StopOldest}
 concurrencySet [/Game/Sounds/Concurrency_Rules/Athena_EmoteFoley_Conc]
 soundSubmixSends [{minSendLevel=0, soundSubmix=&Fortnite_AudioAnalysis_Music}]
}
