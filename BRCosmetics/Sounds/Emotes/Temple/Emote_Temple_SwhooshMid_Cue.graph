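# Sound cue for the mid-range swoosh in the Temple emote, using emote attenuation, sound class and foley concurrency.
# A quality-level node sends its first two levels to a five-wave random pool and its last two to a two-wave pool.
# A root modulator randomizes volume and raises pitch by a random amount.
# The cue also sends to the music audio-analysis submix.
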SoundCue /BRCosmetics/Sounds/Emotes/Temple/Emote_Temple_SwhooshMid_Cue
root=Modulator_0
nodes [
 {id=Modulator_0, children [QualityLevel_0], properties {pitchMax=1.65, pitchMin=1.15, volumeMin=0.65}}
 {id=QualityLevel_0, children [Random_0,Random_0,Random_1,Random_1]}
 {id=Random_0, children [WavePlayer_4, WavePlayer_6, WavePlayer_7, WavePlayer_8, WavePlayer_10], properties {weights [1,1,1,1,1]}}
 {id=Random_1, children [WavePlayer_11, WavePlayer_15], properties {bShouldExcludeFromBranchCulling=true, weights [1,1]}}
 {id=WavePlayer_10, properties {soundWaveAssetPtr=/Game/Athena/Sounds/Emotes/Temple/swoosh_mid_fast_d_02}}
 {id=WavePlayer_11, properties {soundWaveAssetPtr=/Game/Athena/Sounds/Emotes/Temple/swoosh_mid_fast_c_02}}
 {id=WavePlayer_15, properties {soundWaveAssetPtr=&swoosh_mid_fast_d_02}}
 {id=WavePlayer_4, properties {soundWaveAssetPtr=&swoosh_mid_fast_c_02}}
 {id=WavePlayer_6, properties {soundWaveAssetPtr=/Game/Athena/Sounds/Emotes/Temple/swoosh_mid_fast_c_04}}
 {id=WavePlayer_7, properties {soundWaveAssetPtr=/Game/Athena/Sounds/Emotes/Temple/swoosh_mid_fast_c_05}}
 {id=WavePlayer_8, properties {soundWaveAssetPtr=/Game/Athena/Sounds/Emotes/Temple/swoosh_mid_fast_c_06}}
]
properties {
 attenuationSettings=&Fort_Emotes_Athena_Att
 maxDistance=1600
 soundClassObject=&Fort_Music_Emote
 volumeMultiplier=0.45
 concurrencyOverrides {maxCount=3, resolutionRule=StopOldest}
 concurrencySet [/Game/Sounds/Concurrency_Rules/Athena_EmoteFoley_Conc]
 soundSubmixSends [{minSendLevel=0, soundSubmix=&Fortnite_AudioAnalysis_Music}]
}
